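# Test sound cue that plays a 1k test tone wave, with debug enabled and routed to the SFX sound class.
# A single wave player is the root node, with no mixing, modulation or other processing.
# The cue overrides attenuation and concurrency on itself, with concurrency set to prevent new instances once the limit is reached.
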SoundCue /Game/Sounds/Fort_Test_Sounds/Tones/1k_test_tone_Cue
root=WavePlayer_139
nodes [{id=WavePlayer_139, properties {soundWaveAssetPtr=&/1k_test_tone}}]
properties {
 bDebug=true
 bOverrideAttenuation=true
 bOverrideConcurrency=true
 maxDistance=4000
 soundClassObject=&SFX
 volumeMultiplier=1
 concurrencyOverrides {resolutionRule=PreventNew}
}
