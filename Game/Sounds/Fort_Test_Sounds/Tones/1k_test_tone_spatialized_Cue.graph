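# A test SoundCue that plays the 1k test tone in the SFX sound class, named as the spatialized variant and flagged for debug.
# A single Wave Player is the root, with the cue doubling the pitch.
# Inline attenuation overrides turn off distance volume falloff but enable low-pass filter attenuation.
# An inline concurrency override prevents new instances once the limit is reached.
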SoundCue /Game/Sounds/Fort_Test_Sounds/Tones/1k_test_tone_spatialized_Cue
root=WavePlayer_1931
nodes [{id=WavePlayer_1931, properties {soundWaveAssetPtr=&/1k_test_tone}}]
properties {
 bDebug=true
 bOverrideAttenuation=true
 bOverrideConcurrency=true
 maxDistance=2097152
 pitchMultiplier=2
 soundClassObject=&SFX
 volumeMultiplier=1
 attenuationOverrides {bAttenuate=false, bAttenuateWithLPF=true}
 concurrencyOverrides {resolutionRule=PreventNew}
}
