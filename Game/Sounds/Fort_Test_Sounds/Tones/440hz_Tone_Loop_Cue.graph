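# Test sound cue that plays a continuous 440 Hz tone from the Fort_Test_Sounds tones set.
# A single looping wave player on 440hz_Tone_Loop is the root, with the cue volume reduced and a very large max distance.
# The cue also has a send to the Fortnite_AudioAnalysis_Music submix, which by its name feeds music audio analysis.
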SoundCue /Game/Sounds/Fort_Test_Sounds/Tones/440hz_Tone_Loop_Cue
root=WavePlayer_0
nodes [{id=WavePlayer_0, properties {bLooping=true, soundWaveAssetPtr=&/440hz_Tone_Loop}}]
properties {
 maxDistance=2097152
 volumeMultiplier=0.4
 soundSubmixSends [{minSendLevel=0, soundSubmix=&Fortnite_AudioAnalysis_Music}]
}
